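# Physics asset for the TNTina Basketball Hoop gadget skeletal mesh in BRCosmetics.
# Three capsule bodies approximate the hoop: a large one on the Anim bone, a medium one on the rim and a small one on net_01.
# Two constraints chain them from Anim to rim to net_01, each sharing the same swing and twist limit angle.
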PhysicsAsset /BRCosmetics/Gadgets/Assets/TNTina_Basketball_Hoop/Meshes/TNTina_Basketball_Hoop_Gadget_Physics
bodies {
 Anim [{shapeType=Capsule, length=115.1193, radius=77.2225, center=<-6.3151 2.4732 28.015>, rotation=<rot 0 0.0801 -89.1982>}]
 net_01 [{shapeType=Capsule, length=20.6187, radius=20.2308, center=<50.588 -3.2811 2.4944>, rotation=<rot 0 -7.8734 -172.8118>}]
 rim [{shapeType=Capsule, length=35.9399, radius=33.4789, center=<39.5868 2.2218 -10.1611>, rotation=<rot 0 -111.726 -64.2544>}]
}
constraints [
 {bone1Name=rim, bone2Name=Anim, swing1LimitDegrees=45, swing2LimitDegrees=45, twistLimitDegrees=45}
 {bone1Name=net_01, bone2Name=rim, swing1LimitDegrees=45, swing2LimitDegrees=45, twistLimitDegrees=45}
]
